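# Simple item card widget derived from WBP_UIKit_ItemCard_Base, with the equipped and enhanced states cleared in PreConstruct.
# When set as a list entry, it casts the list object to FortItemVM, initializes the card and assigns the viewmodel.
# Its image function uses the item definition's large preview when UseLargeImage is on and that image is valid, otherwise the passed-in image.
# The parent's unhover handling runs only when the owning player is not using touch.
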WidgetBlueprint /FortUILibrary/UIKit/Items/WBP_UIKit_ItemCard_Simple
graphs [
 EventGraph {nodes [
   n1 = Event PreConstruct()
   > Parent: PreConstruct(IsDesignTime=@n1.IsDesignTime)
   > SetItemIsEquipped!(execute, self:WBP UIKit Item Card Base Object Reference, Is Equipped:bool=false) -> (then)
   > SetItemIsEnhanced!(execute, self:WBP UIKit Item Card Base Object Reference, Is Enhanced:bool=false) -> (then)
   n2 = Event OnListItemObjectSet()
   > n3 = Cast To FortItemVM!(execute, Object=@n2.ListItemObject) -> (then, CastFailed, AsFort Item VM:Fort Item VM Object Reference) {is=title K2Node_DynamicCast}
   > InitializeItemCard(self:WBP UIKit Item Card Base Object Reference)
   > SetFortItemVM!(execute, self, Viewmodel:Fort Item VM Object Reference=@n3) -> (then)
   Event On Unhovered()
   > n4 = Branch(Condition=IsUsingTouch(PlayerController=GetOwningPlayer()))
   Parent: On Unhovered(execute=@n4.else) {is=title K2Node_CallParentFunction}
  ]}
 Set Itemcard Image {nodes [
   Set Itemcard Image!() -> (then, Image:Texture 2D Soft Object Reference) {is=title K2Node_FunctionEntry}
   > n1 = Branch(Condition=Get UseLargeImage() -> (UseLargeImage:bool))
   > n2 = Get!(execute:Exec) -> (then:Exec, else:Exec, FortItemVM:Fort Item VM Object Reference)
   n3 = GetLargePreviewImage!(ItemDefinition:Item Definition Base Object Reference=Get ItemDefinition!(self:Fort Item VM Object Reference=@n2) -> (ItemDefinition:Fort Item Definition Object Reference)) -> (ReturnValue:Texture 2D Soft Object Reference)
   > n4 = Branch(Condition=IsValidSoftObjectReference(SoftObjectReference=@n3))
   > SetItemImage!(execute, self:WBP UIKit Item Card Base Object Reference, Item Image:Soft Object Reference=@n3) -> (then)
   SetItemImage(execute=[@n1.else, @n4.else, @n2.else], Item Image=Get Image() -> (Image:Texture 2D Soft Object Reference))
  ]}
]
variables [UseLargeImage {type=boolean}]
widgets {
 info {
  inheritedWidgetCount=0
  parentClass=/FortUILibrary/Systems/Items/WBP_UIKit_ItemCard_Base.WBP_UIKit_ItemCard_Base_C
  rootWidgetClass=null
  widgetCount=0
 }
 widgets []
}
