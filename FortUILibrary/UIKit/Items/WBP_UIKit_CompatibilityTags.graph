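# UIKit widget that shows a product's compatibility tags as wrapped tag widgets beside an optional simple item card.
# SetCompatibilityTags maps gameplay tags through cosmetic configs to unique, non-empty product names, honoring each config's display flag and SkipTagList.
# ShouldSkipTag is never reset, so one skip match drops all later tags.
# AddCompatibilityTags rebuilds the wrap box from those names; PreConstruct uses placeholders at design time, otherwise starts empty with the card collapsed.
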WidgetBlueprint /FortUILibrary/UIKit/Items/WBP_UIKit_CompatibilityTags
graphs [
 AddCompatibilityTags {nodes [
   n1 = AddCompatibilityTags!() -> (then, GameTags:Array of Texts) {is=title K2Node_FunctionEntry}
   > Set Game Tags(Game Tags:Array of Texts=@n1) -> (Output_Get:Array of Texts)
   > ClearChildren(self=Get ModeTagsBox() -> (ModeTagsBox:Wrap Box Object Reference))
   > n2 = For Each Loop(Array:Array of Texts=Get GameTags() -> (GameTags:Array of Texts)) -> (Array Element:Text)
   > Set CurrentArrayElement(CurrentArrayElement:Text=@n2) -> (Output_Get:Text)
   > n3 = Create Widget!(
    execute:Exec
    Class:User Widget Class Reference=Get GameTagClass() -> (GameTagClass:WBP UIKit Tag Base Class Reference)
    OwningPlayer:Player Controller Object Reference=GetOwningPlayer()
   ) -> (then:Exec, ReturnValue:WBP UIKit Tag Base Object Reference)
   > n4 = AddChildToWrapBox!(execute, self:Wrap Box Object Reference=Get ModeTagsBox(), Content:Widget Object Reference=@n3) -> (then, ReturnValue:Wrap Box Slot Object Reference)
   > SetPadding!(execute, self:Wrap Box Slot Object Reference=@n4, InPadding:Margin Structure=Make Margin(Left=0, Top=0, Right=8, Bottom=0)) -> (then)
   > SetTagType!(
    execute
    self:WBP UIKit Tag Base Object Reference=@n3
    Type:E_UI_TagType Enum=Get GameTagType() -> (GameTagType:E_UI_TagType Enum=NewEnumerator0)
   ) -> (then)
   > SetLabelText!(execute, self:WBP UIKit Tag Base Object Reference=@n3, Text:Text) -> (then)
   > SetTagText!(execute, self:WBP UIKit Tag Base Object Reference=@n3, Text:Text=Get CurrentArrayElement() -> (CurrentArrayElement:Text)) -> (then)
   > SetTagSize!(
    execute
    self:WBP UIKit Tag Base Object Reference=@n3
    TagSize:E_UI_TagSize Enum=Get Tag Size() -> (Tag Size:E_UI_TagSize Enum=NewEnumerator0)
   ) -> (then)
  ]}
 EventGraph {nodes [
   n1 = Event PreConstruct()
   > n2 = Branch(Condition=@n1.IsDesignTime)
   > SetupForDesignTime()
   InitializeWidget(execute=@n2.else)
  ]}
 InitializeWidget {nodes [
   InitializeWidget() {is=title K2Node_FunctionEntry}
   > AddCompatibilityTags!(execute, self, GameTags:Array of Texts) -> (then) {is=title K2Node_CallFunction}
   > UpdateItemcardVisibility!(execute, self, ShowItemcard:bool=false) -> (then)
  ]}
 SetCompatibilityTags {nodes [
   n1 = SetCompatibilityTags!() -> (then, Product Tags:Set of Gameplay Tag Structures) {is=title K2Node_FunctionEntry}
   > Clear(TargetArray:Array of Texts=Get Game Tags() -> (Game Tags:Array of Texts))
   > n2 = To Array!(execute, A:Set of Gameplay Tag Structures=@n1) -> (then, Result:Array of Gameplay Tag Structures)
   > n3 = For Each Loop(Array:Array of Gameplay Tag Structures=@n2) -> (Array Element:Gameplay Tag Structure)
   AddCompatibilityTags(execute=@n3.Completed, GameTags=Get Game Tags())
   n4 = For Each Loop with Break!(
    Exec:Exec=@n3.LoopBody
    Array:Array of Gameplay Tag Structures=Get SkipTagList() -> (SkipTagList:Array of Gameplay Tag Structures)
    Break:Exec=@n10.then
   ) -> (LoopBody:Exec, Array Element:Gameplay Tag Structure, Array Index:int, Completed:Exec)
   n5 = Branch(execute=@n4.Completed, Condition=Get ShouldSkipTag() -> (ShouldSkipTag:bool))
   n6 = FindCosmeticConfig!(execute=@n5.else, ProductTag:Gameplay Tag Structure=@n3) -> (then, ReturnValue:Fort Product Cosmetic Config Object Reference)
   > Is Valid(InputObject=@n6)
   n7 = Get ProductName!(self:Fort Product Cosmetic Config Object Reference=@n6) -> (ProductName:Text)
   n8 = TextIsEmpty(InText=@n7)
   > Branch(Condition=AND Boolean(
     A=NOT Boolean(A=@n8)
     B=Get bDisplayCompatibilityTag!(self:Fort Product Cosmetic Config Object Reference=@n6) -> (bDisplayCompatibilityTag:bool)
    ))
   > Add Unique(TargetArray:Array of Texts=Get Game Tags(), NewItem:"Text (by ref)"=@n7)
   n9 = "Equal (GameplayTag)"(A=@n4, B=@n3)
   Branch(execute=@n4.LoopBody, Condition=@n9)
   > n10 = Set ShouldSkipTag(ShouldSkipTag:bool=@n9) -> (Output_Get:bool)
  ]}
 SetupForDesignTime {nodes [
   SetupForDesignTime() {is=title K2Node_FunctionEntry}
   > AddCompatibilityTags(GameTags=Get PHGameplayTags() -> (PHGameplayTags:Array of Texts))
   > SetVisibility(
    self=Get Itemcard_Simple() -> (Itemcard_Simple:WBP UIKit Item Card Simple Object Reference)
    InVisibility=Select(Option 0=HitTestInvisible, Option 1=Collapsed, Index=Get HideItemcard() -> (HideItemcard:bool))
   )
  ]}
 UpdateItemcardVisibility {nodes [
   n1 = UpdateItemcardVisibility!() -> (then, ShowItemcard:bool=~) {is=title K2Node_FunctionEntry}
   n2 = NOT Boolean(A=@n1)
   n3 = Get HideItemcard()
   > SetVisibility(self=Get Itemcard_Simple(), InVisibility=Select(Option 0=HitTestInvisible, Option 1=Collapsed, Index=OR Boolean(A=@n2, B=@n3)))
  ]}
 VM_BoolToTagSize {nodes [
   n1 = VM_BoolToTagSize!() -> (then, InBool:bool=~, True Tag Size:E_UI_TagSize Enum=NewEnumerator0, False Tag Size:E_UI_TagSize Enum=NewEnumerator0) {is=title K2Node_FunctionEntry}
   > n2 = Branch(Condition=@n1)
   > Return Node!(execute, ReturnValue:E_UI_TagSize Enum=Get True Tag Size() -> (True Tag Size:E_UI_TagSize Enum=NewEnumerator0))
   Return Node(execute=@n2.else, ReturnValue=Get False Tag Size() -> (False Tag Size:E_UI_TagSize Enum=NewEnumerator0))
  ]}
]
variables [
 Game Tags {type="array<string>", default []}
 GameTagClass {type=Class, default=/FortUILibrary/UIKit/Tag/WBP_UIKit_Tag.WBP_UIKit_Tag_C}
 GameTagType {type=E_UI_TagType, default=Secondary Tag}
 HideItemcard {type=boolean}
 ShouldSkipTag {type=boolean}
 SkipTagList {type="array<GameplayTag>", default []}
 Tag Size {type=E_UI_TagSize, default=Compact}
]
widgets {
 info {inheritedWidgetCount=0, parentClass=&UserWidget, rootWidgetClass=&Overlay, widgetCount=4}
 widgets [
  Overlay_39 {widgetClassPath=&Overlay, widget=Overlay_39}
  HorizontalBox_0 {
   widgetClassPath=&HorizontalBox
   parent=Overlay_39
   widget=HorizontalBox_0
   properties {flowDirectionPreference=Culture}
   slot {class=OverlaySlot, properties {horizontalAlignment=HAlign_Fill}}
  }
  Itemcard_Simple {
   widgetClassPath=&/WBP_UIKit_ItemCard_Simple
   bIsVariable=true
   parent=HorizontalBox_0
   widget=Itemcard_Simple
   properties {
    bSelectable=false
    holdData=&CommonUI_HoldData_Base_C
    visibility=HitTestInvisible
    itemCardSize=<100 100>
   }
   slot {class=HorizontalBoxSlot, properties {horizontalAlignment=HAlign_Left, verticalAlignment=VAlign_Center, padding=<ltrb 0 0 16 0>}}
  }
  ModeTagsBox {
   widgetClassPath=&WrapBox
   bIsVariable=true
   parent=HorizontalBox_0
   widget=ModeTagsBox
   properties {innerSlotPadding=<0 8>}
   slot {class=HorizontalBoxSlot, properties {verticalAlignment=VAlign_Center, size {sizeRule=Fill, value=1}}}
  }
 ]
}
