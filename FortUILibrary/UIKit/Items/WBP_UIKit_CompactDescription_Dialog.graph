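# Activatable item dialog that explains which experiences an item works with, showing formatted compatibility text and tags in a scroll box.
# CreateCustomWidget async-loads an optional extra widget into CustomSlot and passes it the FortItemDetailsVM, collapsing the slot for an invalid class.
# Activation logs an analytics sub-view event with ItemName and binds the dialog button, which logs a close and deactivates.
# SetMobileStyle adjusts dialog height and button fill for compact UI.
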WidgetBlueprint /FortUILibrary/UIKit/Items/WBP_UIKit_CompactDescription_Dialog
graphs [
 EventGraph {nodes [
   n1 = OnButtonClicked!() -> (OutputDelegate, then, ButtonIndex:int) {is=title K2Node_CustomEvent}
   > FireEventElementInteractedByName!(
    execute
    Element Name:String=Button_Close
    WidgetCallingFrom:Widget Object Reference=Get WBP_UIKit_Dialog_Internal() -> (WBP_UIKit_Dialog_Internal:WBP UIKit Dialog Internal Object Reference)
   ) -> (then, Success:bool)
   > DeactivateWidget()
   n2 = CreateCustomWidget!() -> (OutputDelegate, then, UserWidgetSoftClass:User Widget Soft Class Reference) {is=title K2Node_CustomEvent}
   > n3 = Branch(Condition=IsValidSoftClassReference!(SoftClassReference:"Object Soft Class Reference (by ref)"=@n2.UserWidgetSoftClass) -> (ReturnValue:bool))
   > SetVisibility(self=Get CustomSlot() -> (CustomSlot:Overlay Object Reference), InVisibility=SelfHitTestInvisible)
   > ClearChildren(self=Get CustomSlot())
   > n4 = CreateWidgetAsync!(
    execute:Exec
    UserWidgetSoftClass:User Widget Soft Class Reference=@n2.UserWidgetSoftClass
    OwningPlayer:Player Controller Object Reference=GetOwningPlayer()
    bSuspendInputUntilComplete:bool=false
   ) -> (then:Exec, OnComplete:Exec, OnFailure:Exec, UserWidget:User Widget Object Reference) {is=title K2Node_AsyncAction}
   AddChild(execute=@n4.OnComplete, self=Get CustomSlot(), Content=@n4)
   > SetViewModelByClass!(
    execute
    Widget:User Widget Object Reference=@n4
    ViewModel:Notify Field Value Changed Interface=Get FortItemDetailsVM() -> (FortItemDetailsVM:Fort Item Details VM Object Reference)
   ) -> (then)
   SetVisibility(execute=@n3.else, self=Get CustomSlot(), InVisibility=Collapsed)
   Event On Activated()
   > Fire Event Sub View Opened!(
    execute
    View:Widget Object Reference=Self-Reference()
    AdditionalInfo:Flow Analytics Additional Info Structure="Add String (Non-Localized)"!(
     Info:"Flow Analytics Additional Info Structure (by ref)"=Make Flow Analytics Additional Info!() -> (FlowAnalyticsAdditionalInfo:Flow Analytics Additional Info Structure) {is=title K2Node_MakeStruct}
     Key:String=Item
     Value:String=Get ItemName() -> (ItemName:String)
    ) -> (ReturnValue:Flow Analytics Additional Info Structure)
    bIsOverlay:bool=true
   ) -> (then)
   > Bind Event to Dialog Button Clicked Dispatcher(self:WBP UIKit Dialog Internal Object Reference=Get WBP_UIKit_Dialog_Internal(), Delegate=@n1) {is=title K2Node_AddDelegate}
   Event On Deactivated()
   > Unbind Event from Dialog Button Clicked Dispatcher(self:WBP UIKit Dialog Internal Object Reference=Get WBP_UIKit_Dialog_Internal(), Delegate=@n1) {is=title K2Node_RemoveDelegate}
   Event PreConstruct()
  ]}
 GetCompatabilityDetails {nodes [
   n1 = GetCompatabilityDetails!() -> (then, ItemName:Text, ItemType:Text) {is=title K2Node_FunctionEntry}
   > Return Node!(
    execute
    CompatibilityInfo:Text=Format Text!(
     Format="LOCTEXT(\"Not all items work in all experiences.\\r\\n\\r\\n<text color=\\\"FFF\\\" fontface=\\\"76 Bold\\\">{name}</> <text color=\\\"FFF\\\" case=\\\"lower\\\">{type}</> works with:\")"
     name:Text=@n1
     type:Text=@n1.ItemType
    ) -> (Result)
   )
  ]}
 SetCompatibilityString {nodes [
   n1 = SetCompatibilityString!() -> (then, Dialog Info Text:Text) {is=title K2Node_FunctionEntry}
   > SetText(self=Get Text_Compatibility() -> (Text_Compatibility:Common Rich Text Block Object Reference), InText=@n1)
  ]}
 SetMobileStyle {nodes [
   SetMobileStyle()
   n1 = IsUsingCompactUI()
   n2 = Get WBP_UIKit_Dialog_Internal()
   n3 = Get!(self:WBP UIKit Dialog Internal Object Reference=@n2) -> (
    DialogSIze_WidthOverride_107_#1:S UI Bool Float Structure
    DialogSIze_HeightOverride_108_#2:S UI Bool Float Structure
    DialogSIze_MinDesiredWidth_109_#3:S UI Bool Float Structure
    DialogSIze_MinDesiredHeight_110_#4:S UI Bool Float Structure
    DialogSIze_MaxDesiredWidth_111_#5:S UI Bool Float Structure
    DialogSIze_MaxDesiredHeight_112_#6:S UI Bool Float Structure
    DialogSIze_MinAspectRatio_113_#7:S UI Bool Float Structure
    DialogSIze_MaxAspectRatio_114_#8:S UI Bool Float Structure
   )
   > Set!(
    execute
    DialogSIze_WidthOverride_107_#1:S UI Bool Float Structure=@n3
    DialogSIze_HeightOverride_108_#2:S UI Bool Float Structure=Select(
     Option 0:S UI Bool Float Structure=Make S UI Bool Float!(Enable_3_#9:bool=true, Value_4_#10:double=650) -> (S_UI_BoolFloat:S UI Bool Float Structure) {is=title K2Node_MakeStruct}
     Option 1:S UI Bool Float Structure=Make S UI Bool Float(Enable_3_#9=False, Value_4_#10=0)
     Index=@n1
    ) -> (ReturnValue:S UI Bool Float Structure)
    DialogSIze_MinDesiredWidth_109_#3:S UI Bool Float Structure=@n3.DialogSIze_MinDesiredWidth_109_#3
    DialogSIze_MinDesiredHeight_110_#4:S UI Bool Float Structure=@n3.DialogSIze_MinDesiredHeight_110_#4
    DialogSIze_MaxDesiredWidth_111_#5:S UI Bool Float Structure=@n3.DialogSIze_MaxDesiredWidth_111_#5
    DialogSIze_MaxDesiredHeight_112_#6:S UI Bool Float Structure=@n3.DialogSIze_MaxDesiredHeight_112_#6
    DialogSIze_MinAspectRatio_113_#7:S UI Bool Float Structure=@n3.DialogSIze_MinAspectRatio_113_#7
    DialogSIze_MaxAspectRatio_114_#8:S UI Bool Float Structure=@n3.DialogSIze_MaxAspectRatio_114_#8
    self:WBP UIKit Dialog Internal Object Reference=@n2
   ) -> (then, Output_Get:S UI Size Box Structure)
   n4 = Get!(self=@n2) -> (
    Buttons_GroupPlacement_71_#11:S UI Placement Structure
    Buttons_SpaceBetweenButtons_62_#12:double=~
    Buttons_EachButtonSize_90_#13:E_UI_ButtonSize Enum
    Buttons_EachButtonMinWidth_77_#14:int
    Buttons_EachButtonHorizontalFill_87_#15:bool=~
    Buttons_CTAButtonsTypes_89_#16:Array of E_UI_ButtonVisualType Enums
    Buttons_FocuseButton_85_#17:int
   )
   > Set!(
    execute
    Buttons_GroupPlacement_71_#11:S UI Placement Structure=@n4
    Buttons_SpaceBetweenButtons_62_#12:double=@n4.Buttons_SpaceBetweenButtons_62_#12
    Buttons_EachButtonSize_90_#13:E_UI_ButtonSize Enum=@n4.Buttons_EachButtonSize_90_#13
    Buttons_EachButtonMinWidth_77_#14:int=@n4.Buttons_EachButtonMinWidth_77_#14
    Buttons_EachButtonHorizontalFill_87_#15:bool=NOT Boolean(A=@n1)
    Buttons_CTAButtonsTypes_89_#16:Array of E_UI_ButtonVisualType Enums=@n4.Buttons_CTAButtonsTypes_89_#16
    Buttons_FocuseButton_85_#17:int=@n4.Buttons_FocuseButton_85_#17
    self=@n2
   ) -> (then, Output_Get:S UI Dialog Buttons Structure)
  ]}
]
variables [
 CustomModalContent {type=Widget, default=null}
 ItemName {type=string, default=""}
]
widgets {
 info {
  inheritedWidgetCount=0
  namedSlotCount=1
  parentClass=&CommonActivatableWidget
  rootWidgetClass=/FortUILibrary/Systems/Dialog/WBP_UIKit_Dialog_Internal.WBP_UIKit_Dialog_Internal_C
  widgetCount=8
 }
 widgets [
  WBP_UIKit_Dialog_Internal {
   widgetClassPath=/FortUILibrary/Systems/Dialog/WBP_UIKit_Dialog_Internal
   bIsVariable=true
   widget=WBP_UIKit_Dialog_Internal
   properties {
    designTimePreview_DialogVM=null
    useText=Common Rich Text
    buttons {
     each Button Horizontal Fill=false
     each Button Min Width=0
     each Button Size=Medium
     focuse Button=0
     space Between Buttons=30
     cTA Buttons Types [Regular]
     group Placement {
      horizontal Alignment=HAlign_Fill
      vertical Alignment=VAlign_Bottom
      grid {column=1, column Span=0, layer=0, row=2, row Span=0, nudge=<0 0>}
      padding=<ltrb 80 64 80 48>
     }
    }
    dialogSIze {
     height Override {enable=false, value=800}
     max Aspect Ratio {enable=false, value=0}
     max Desired Height {enable=true, value=800}
     max Desired Width {enable=false, value=1000}
     min Aspect Ratio {enable=false, value=0}
     min Desired Height {enable=false, value=0}
     min Desired Width {enable=false, value=800}
     width Override {enable=true, value=800}
    }
   }
  }
  Overlay_0 {widgetClassPath=&Overlay, namedSlotHost=WBP_UIKit_Dialog_Internal, widget=Overlay_0}
  ScrollBox_85 {
   widgetClassPath=&ScrollBox
   parent=Overlay_0
   widget=ScrollBox_85
   properties {
    bAnimateWheelScrolling=true
    scrollbarPadding=<ltrb 8 0 0 0>
    scrollbarThickness=<12 12>
    widgetStyle {
     barThickness=12
     bottomShadowBrush {
      bIsDynamicallyLoaded=false
      drawAs=NoDrawType
      imageType=FullColor
      mirroring=NoMirror
      resourceName=../../../Engine/Content/Slate/Common/ScrollBoxShadowBottom.png
      resourceObject=null
      tiling=NoTile
      imageSize=<16 8>
      margin=<ltrb 0.5 0 0.5 1>
      outlineSettings {
       bUseBrushTransparency=false
       roundingType=HalfHeightRadius
       width=0
       color {colorUseRule=UseColor_Specified, specifiedColor=<rgba 0 0 0 0>}
       cornerRadii=<0 0 0 0>
      }
      tintColor {colorUseRule=UseColor_Specified, specifiedColor=<rgba 1 1 1 1>}
      uVRegion {bIsValid=false, max=<0 0>, min=<0 0>}
     }
     horizontalScrolledContentPadding=<ltrb 0 0 1 0>
     leftShadowBrush {
      bIsDynamicallyLoaded=false
      drawAs=NoDrawType
      imageType=FullColor
      mirroring=NoMirror
      resourceName=../../../Engine/Content/Slate/Common/ScrollBoxShadowLeft.png
      resourceObject=null
      tiling=NoTile
      imageSize=<8 16>
      margin=<ltrb 1 0.5 0 0.5>
      outlineSettings {
       bUseBrushTransparency=false
       roundingType=HalfHeightRadius
       width=0
       color {colorUseRule=UseColor_Specified, specifiedColor=<rgba 0 0 0 0>}
       cornerRadii=<0 0 0 0>
      }
      tintColor {colorUseRule=UseColor_Specified, specifiedColor=<rgba 1 1 1 1>}
      uVRegion {bIsValid=false, max=<0 0>, min=<0 0>}
     }
     rightShadowBrush {
      bIsDynamicallyLoaded=false
      drawAs=NoDrawType
      imageType=FullColor
      mirroring=NoMirror
      resourceName=../../../Engine/Content/Slate/Common/ScrollBoxShadowRight.png
      resourceObject=null
      tiling=NoTile
      imageSize=<8 16>
      margin=<ltrb 0 0.5 1 0.5>
      outlineSettings {
       bUseBrushTransparency=false
       roundingType=HalfHeightRadius
       width=0
       color {colorUseRule=UseColor_Specified, specifiedColor=<rgba 0 0 0 0>}
       cornerRadii=<0 0 0 0>
      }
      tintColor {colorUseRule=UseColor_Specified, specifiedColor=<rgba 1 1 1 1>}
      uVRegion {bIsValid=false, max=<0 0>, min=<0 0>}
     }
     topShadowBrush {
      bIsDynamicallyLoaded=false
      drawAs=NoDrawType
      imageType=FullColor
      mirroring=NoMirror
      resourceName=../../../Engine/Content/Slate/Common/ScrollBoxShadowTop.png
      resourceObject=null
      tiling=NoTile
      imageSize=<16 8>
      margin=<ltrb 0.5 1 0.5 0>
      outlineSettings {
       bUseBrushTransparency=false
       roundingType=HalfHeightRadius
       width=0
       color {colorUseRule=UseColor_Specified, specifiedColor=<rgba 0 0 0 0>}
       cornerRadii=<0 0 0 0>
      }
      tintColor {colorUseRule=UseColor_Specified, specifiedColor=<rgba 1 1 1 1>}
      uVRegion {bIsValid=false, max=<0 0>, min=<0 0>}
     }
     verticalScrolledContentPadding=<ltrb 0 0 0 1>
    }
   }
   slot {class=OverlaySlot, properties {horizontalAlignment=HAlign_Fill, padding=<ltrb 80 0 80 0>}}
  }
  VerticalBox_494 {widgetClassPath=&VerticalBox, parent=ScrollBox_85, widget=VerticalBox_494, slot {class=ScrollBoxSlot}}
  Text_Compatibility {
   widgetClassPath=&CommonRichTextBlock
   bIsVariable=true
   parent=VerticalBox_494
   widget=Text_Compatibility
   properties {
    autoWrapText=true
    bOverrideDefaultStyle=true
    text="Not all items work in all experiences.\r\n\r\nThis item works with:"
    defaultTextStyleOverride {
     overflowPolicy=Clip
     transformPolicy=None
     colorAndOpacity {colorUseRule=UseColor_Specified, specifiedColor=<rgba 1 1 1 1>}
     font {
      bForceMonospaced=false
      bMaterialIsStencil=false
      fontMaterial=null
      fontObject=&HeadingNow
      letterSpacing=0
      monospacedWidth=1
      size=15.75
      skewAmount=0
      typefaceFontName=74 Regular
      outlineSettings {
       bApplyOutlineToDropShadows=false
       bMiteredCorners=false
       bSeparateFillAlpha=false
       outlineMaterial=null
       outlineSize=0
       outlineColor=<rgba 0 0 0 1>
      }
     }
     highlightColor {colorUseRule=UseColor_Specified, specifiedColor=<rgba 0 0 0 1>}
     highlightShape {
      bIsDynamicallyLoaded=false
      drawAs=Image
      imageType=NoImage
      mirroring=NoMirror
      resourceName=None
      resourceObject=null
      tiling=NoTile
      imageSize=<32 32>
      margin=<ltrb 0 0 0 0>
      outlineSettings {
       bUseBrushTransparency=false
       roundingType=HalfHeightRadius
       width=0
       color {colorUseRule=UseColor_Specified, specifiedColor=<rgba 0 0 0 0>}
       cornerRadii=<0 0 0 0>
      }
      tintColor {colorUseRule=UseColor_Specified, specifiedColor=<rgba 1 1 1 1>}
      uVRegion {bIsValid=false, max=<0 0>, min=<0 0>}
     }
     selectedBackgroundColor {colorUseRule=UseColor_Foreground, specifiedColor=<rgba 1 0 1 1>}
     shadowColorAndOpacity=<rgba 0 0 0 1>
     shadowOffset=<0 0>
     strikeBrush {
      bIsDynamicallyLoaded=false
      drawAs=Image
      imageType=NoImage
      mirroring=NoMirror
      resourceName=None
      resourceObject=null
      tiling=NoTile
      imageSize=<32 32>
      margin=<ltrb 0 0 0 0>
      outlineSettings {
       bUseBrushTransparency=false
       roundingType=HalfHeightRadius
       width=0
       color {colorUseRule=UseColor_Specified, specifiedColor=<rgba 0 0 0 0>}
       cornerRadii=<0 0 0 0>
      }
      tintColor {colorUseRule=UseColor_Specified, specifiedColor=<rgba 1 1 1 1>}
      uVRegion {bIsValid=false, max=<0 0>, min=<0 0>}
     }
     underlineBrush {
      bIsDynamicallyLoaded=false
      drawAs=Image
      imageType=NoImage
      mirroring=NoMirror
      resourceName=None
      resourceObject=null
      tiling=NoTile
      imageSize=<32 32>
      margin=<ltrb 0 0 0 0>
      outlineSettings {
       bUseBrushTransparency=false
       roundingType=HalfHeightRadius
       width=0
       color {colorUseRule=UseColor_Specified, specifiedColor=<rgba 0 0 0 0>}
       cornerRadii=<0 0 0 0>
      }
      tintColor {colorUseRule=UseColor_Specified, specifiedColor=<rgba 1 1 1 1>}
      uVRegion {bIsValid=false, max=<0 0>, min=<0 0>}
     }
    }
   }
   slot {class=VerticalBoxSlot, properties {padding=<ltrb 0 0 0 32>}}
  }
  List_Compatibility {
   widgetClassPath=&HorizontalBox
   parent=VerticalBox_494
   widget=List_Compatibility
   slot {class=VerticalBoxSlot, properties {verticalAlignment=VAlign_Top, padding=<ltrb 0 0 0 16>}}
  }
  WBP_UIKit_CompatibilityTags {
   widgetClassPath=&/WBP_UIKit_CompatibilityTags
   bIsVariable=true
   parent=List_Compatibility
   widget=WBP_UIKit_CompatibilityTags
   slot {class=HorizontalBoxSlot, properties {verticalAlignment=VAlign_Top, size {sizeRule=Fill, value=1}}}
  }
  CustomSlot {
   widgetClassPath=&Overlay
   bIsVariable=true
   parent=VerticalBox_494
   widget=CustomSlot
   properties {bOverride_Cursor=true, visibility=Collapsed}
   slot {class=VerticalBoxSlot}
  }
 ]
}
