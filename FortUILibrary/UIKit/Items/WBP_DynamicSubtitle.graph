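# UI kit widget for a subtitle line, built on UserWidget with a single CommonTextBlock, Text_Subtitle, as its root.
# PreConstruct passes the SubtitleFont variable to SetSubtitleFont, which applies it to the text block, then sets placeholder text at design time only.
# The text block displays in uppercase, truncates overflow with an ellipsis and auto-collapses when empty; no runtime text assignment is defined here.
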WidgetBlueprint /FortUILibrary/UIKit/Items/WBP_DynamicSubtitle
graphs [
 EventGraph {nodes [
   n1 = Event PreConstruct()
   > SetSubtitleFont!(execute, self, Font:Slate Font Info Structure=Get SubtitleFont() -> (SubtitleFont:Slate Font Info Structure)) -> (then)
   > Branch(Condition=@n1.IsDesignTime)
   > "SetText (Text)"(self=Get Text_Subtitle() -> (Text_Subtitle:Common Text Object Reference), InText="INVTEXT(\"[PH] Dynamic Subtitle\")")
  ]}
 SetSubtitleFont {nodes [
   n1 = SetSubtitleFont!() -> (then, Font:Slate Font Info Structure) {is=title K2Node_FunctionEntry}
   > Set Font!(execute, Font:Slate Font Info Structure=@n1, self:Text Object Reference=Get Text_Subtitle()) -> (then, Output_Get:Slate Font Info Structure)
  ]}
]
variables [SubtitleFont {
  type=SlateFontInfo
  default {
   bForceMonospaced=false
   bMaterialIsStencil=false
   fontMaterial=null
   fontObject=&HeadingNow
   letterSpacing=0
   monospacedWidth=1
   size=18
   skewAmount=0
   typefaceFontName=HeadingNow-86BoldItalic
   outlineSettings {
    bApplyOutlineToDropShadows=false
    bMiteredCorners=false
    bSeparateFillAlpha=false
    outlineMaterial=null
    outlineSize=0
    outlineColor=<rgba 0 0 0 1>
   }
  }
 }]
widgets {
 info {inheritedWidgetCount=0, parentClass=&UserWidget, rootWidgetClass=&CommonTextBlock, widgetCount=1}
 widgets [Text_Subtitle {
   widgetClassPath=&CommonTextBlock
   bIsVariable=true
   widget=Text_Subtitle
   properties {
    bAutoCollapseWithEmptyText=true
    bIsScrollingEnabled=false
    scrollStyle=&TextScrollStyle-Base_C
    textOverflowPolicy=Ellipsis
    textTransformPolicy=ToUpper
    font {
     bForceMonospaced=false
     bMaterialIsStencil=false
     fontMaterial=null
     fontObject=&HeadingNow
     letterSpacing=0
     monospacedWidth=1
     size=18
     skewAmount=0
     typefaceFontName=86 Bold
     outlineSettings {
      bApplyOutlineToDropShadows=false
      bMiteredCorners=false
      bSeparateFillAlpha=false
      outlineMaterial=null
      outlineSize=0
      outlineColor=<rgba 0 0 0 1>
     }
    }
    shadowOffset=<0 0>
   }
  }]
}
