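# Item description panel widget that extends WBP_UIKit_ItemDescription_Base with a scrolling subtitle and rarity tag handling.
# PreConstruct creates the dynamic subtitle in code and adds it to the inherited grid, mirroring its alignment for Arabic.
# Subtitles show only for SparksSong items, or for CosmeticMimosa items whose custom name, used as title, differs from the item subtitle.
# Rarity tag colors derive from the view model's palette, series flag and tag type.
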WidgetBlueprint /FortUILibrary/UIKit/Items/WBP_UIKit_DescriptionWithSubtitle
graphs [
 EventGraph {nodes [
   n1 = Event PreConstruct()
   > MakeDynamicSubtitle()
   > Parent: PreConstruct(IsDesignTime=@n1.IsDesignTime)
   > Branch(Condition=@n1.IsDesignTime)
   > SetRarityTag!(
    execute
    self
    TagPrimaryText:Text="INVTEXT(\"[PH] SONG\")"
    TagSecondaryText:Text="INVTEXT(\"| [PH] 2019 | 03:28\")"
    TagType:E_UI_TagType Enum=Get RarityTagType() -> (RarityTagType:E_UI_TagType Enum=NewEnumerator0)
   ) -> (then)
   > SetSubtitle!(execute, self, SubtitleText:Text="INVTEXT(\"[PH] Subtitle text\")") -> (then)
  ]}
 GetMimosaName {nodes [
   GetMimosaName() {is=title K2Node_FunctionEntry}
   > n1 = IsNameable!(execute:Exec, self:Object Reference=Get FortItemDetailsVM() -> (FortItemDetailsVM:Fort Item Details VM Object Reference)) -> (then:Exec, ReturnValue:bool) {is=title K2Node_Message}
   > n2 = Branch(Condition=@n1)
   > n3 = GetItemName!(execute:Exec, self:Object Reference=Get FortItemDetailsVM()) -> (then:Exec, ReturnValue:Text) {is=title K2Node_Message}
   > Return Node!(execute, NameText:Text=@n3, IsNameEmpty:bool=TextIsEmpty(InText=@n3), IsNamable:bool=true)
   Return Node(execute=@n2.else, IsNameEmpty=true, IsNamable=false)
  ]}
 GetRarityColors {nodes [
   n1 = GetRarityColors!() -> (then, IsSeries:bool=~, FortColorPalette:"Fort Color Palette Structure (by ref)") {is=title K2Node_FunctionEntry}
   n2 = Get RarityTagType() -> (RarityTagType=NewEnumerator0)
   n3 = Break Fort Color Palette!(FortColorPalette:"Fort Color Palette Structure (by ref)"=@n1.FortColorPalette) -> (
    Color1:Linear Color Structure
    Color2:Linear Color Structure
    Color3:Linear Color Structure
    Color4:Linear Color Structure
    Color5:Linear Color Structure
   ) {is=title K2Node_BreakStruct}
   n4 = SelectColor(A=@n3, B=@n3.Color2, bPickA=@n1)
   > Return Node!(
    execute
    OutputPin:Fort Color Palette Structure=Select(
     Option 0:Fort Color Palette Structure=Make Fort Color Palette!(
      Color1:Linear Color Structure=@n4
      Color2:Linear Color Structure=@n3.Color4
      Color3:Linear Color Structure=(R=0,G=0,B=0,A=1)
      Color4:Linear Color Structure=(R=0,G=0,B=0,A=1)
      Color5:Linear Color Structure=(R=0,G=0,B=0,A=1)
     ) -> (FortColorPalette:Fort Color Palette Structure) {is=title K2Node_MakeStruct}
     Option 1:Fort Color Palette Structure=Make Fort Color Palette(Color1=(R=1,G=1,B=1,A=0.7), Color2=(R=1,G=1,B=1,A=0.01), Color3=(R=0,G=0,B=0,A=1), Color4=(R=0,G=0,B=0,A=1), Color5=(R=0,G=0,B=0,A=1))
     Index="Equal (Enum)"(A:E_UI_TagType Enum=@n2, B:E_UI_TagType Enum=NewEnumerator3)
    ) -> (ReturnValue:Fort Color Palette Structure)
   )
  ]}
 InitializeWidget {nodes [
   InitializeWidget() {is=title K2Node_FunctionEntry}
   > Parent: InitializeWidget() {is=title K2Node_CallParentFunction}
   > n1 = Branch(Condition=Get HideSubtitle() -> (HideSubtitle:bool))
   n2 = Get Subtitle() -> (Subtitle:WBP Dynamic Subtitle Object Reference)
   > SetVisibility(self=@n2, InVisibility=Collapsed)
   n3 = Get Use Rich Description() -> (Use Rich Description:bool)
   n4 = Get HideSetText() -> (HideSetText:bool)
   > SetSetTextVisibility!(execute, self:WBP UIKit Item Description Base Object Reference, Show Text:bool=NOT Boolean(A=OR Boolean(A=@n4, B=@n3))) -> (then)
   n5 = Get Use Rich Description()
   n6 = Get HideSeasonText() -> (HideSeasonText:bool)
   > SetSeasonTextVisibility!(execute, self:WBP UIKit Item Description Base Object Reference, Show Text:bool=NOT Boolean(A=OR Boolean(A=@n6, B=@n5))) -> (then)
   n7 = Get Use Rich Description()
   n8 = Get HideGameplayTagsText() -> (HideGameplayTagsText:bool)
   > SetGameplayTagsTextVisibility!(execute, self:WBP UIKit Item Description Base Object Reference, Show Text:bool=NOT Boolean(A=OR Boolean(A=@n8, B=@n7))) -> (then)
   SetVisibility(execute=@n1.else, self=@n2, InVisibility=SelfHitTestInvisible)
  ]}
 IsMimosa {nodes [
   Inputs!()
   n1 = Get ItemDefinition!(self:Fort Item VM Object Reference=Get Item!(self:Fort Item Details VM Object Reference=Get FortItemDetailsVM()) -> (Item:Fort Item VM Object Reference)) -> (ItemDefinition:Fort Item Definition Object Reference)
   Outputs!(Bool:bool="Equal (Enum)"(
     A:EFortItemType Enum=GetItemType!(ItemDefinition:Item Definition Base Object Reference=@n1) -> (ReturnValue:EFortItemType Enum=WorldItem)
     B:EFortItemType Enum=CosmeticMimosa
    ))
  ]}
 MakeDynamicSubtitle {nodes [
   MakeDynamicSubtitle() {is=title K2Node_FunctionEntry}
   > n1 = Is Valid(InputObject=Get Subtitle())
   n2 = Create Widget!(
    execute:Exec=@n1.Is Not Valid
    Class:User Widget Class Reference=&/WBP_DynamicSubtitle.WBP_DynamicSubtitle_C
    OwningPlayer:Player Controller Object Reference
   ) -> (then:Exec, ReturnValue:WBP Dynamic Subtitle Object Reference)
   > n3 = Set Subtitle(Subtitle:WBP Dynamic Subtitle Object Reference=@n2) -> (Output_Get:WBP Dynamic Subtitle Object Reference)
   n4 = Get Text_Subtitle!(self:WBP Dynamic Subtitle Object Reference=@n3) -> (Text_Subtitle:Common Text Object Reference)
   > SetScrollingEnabled!(execute, self:Common Text Object Reference=@n4, bInIsScrollingEnabled:bool=true) -> (then)
   > Set PixelSnapping!(execute, PixelSnapping:EWidgetPixelSnapping Enum=Disabled, self:Widget Object Reference=@n4) -> (then, Output_Get:EWidgetPixelSnapping Enum=Inherit)
   > AddChildToGrid!(
    execute
    self:Grid Panel Object Reference=Get Root() -> (Root:Grid Panel Object Reference)
    Content:Widget Object Reference=@n3
    InRow:int=Get SubtitleRow() -> (SubtitleRow:int=0)
    InColumn:int=0
   ) -> (then, ReturnValue:Grid Slot Object Reference)
  ]}
 Set Alignments {nodes [
   n1 = Set Alignments!() -> (then, Alignment:EHorizontalAlignment Enum=HAlign_Fill) {is=title K2Node_FunctionEntry}
   > Parent: Set Alignments!(execute, Alignment:EHorizontalAlignment Enum=@n1) -> (then) {is=title K2Node_CallParentFunction}
   n2 = GetCurrentCulture!() -> (ReturnValue:String)
   n3 = Get HorizontalAlignment() -> (HorizontalAlignment:EHorizontalAlignment Enum=HAlign_Fill)
   > Set HorizontalAlignment!(
    execute
    HorizontalAlignment:EHorizontalAlignment Enum=Select(
     Option 0:EHorizontalAlignment Enum=@n3
     Option 1:EHorizontalAlignment Enum=Select!(
      HAlign_Fill:EHorizontalAlignment Enum=HAlign_Fill
      HAlign_Left:EHorizontalAlignment Enum=HAlign_Right
      HAlign_Center:EHorizontalAlignment Enum=HAlign_Center
      HAlign_Right:EHorizontalAlignment Enum=HAlign_Left
      Index:EHorizontalAlignment Enum=@n3
     ) -> (ReturnValue)
     Index="Equal, Case Insensitive (String)"(A=@n2, B=ar)
    ) -> (ReturnValue:EHorizontalAlignment Enum)
    self:Grid Slot Object Reference=SlotAsGridSlot(Widget=Get Subtitle())
   ) -> (then, Output_Get:EHorizontalAlignment Enum=HAlign_Fill)
  ]}
 SetIsEquipped {nodes [
   n1 = SetIsEquipped!() -> (then, IsEquipped:bool=~) {is=title K2Node_FunctionEntry}
   n2 = Get CanShowEquippedTag() -> (CanShowEquippedTag:bool)
   > SetVisibility(
    self=Get EquippedTag() -> (EquippedTag:WBP UIKit Tag Object Reference)
    InVisibility=Select!(Option 0:ESlateVisibility Enum=Collapsed, Option 1:ESlateVisibility Enum=HitTestInvisible, Index:bool=AND Boolean(A=@n1, B=@n2)) -> (ReturnValue:ESlateVisibility Enum)
   )
  ]}
 SetRarityTag {nodes [
   n1 = SetRarityTag!() -> (then, TagPrimaryText:Text, TagSecondaryText:Text, TagType:E_UI_TagType Enum=NewEnumerator0) {is=title K2Node_FunctionEntry}
   > SetRarityTagType!(execute, self, RarityTagType:E_UI_TagType Enum=@n1.TagType) -> (then)
   > SetTagText!(
    execute
    self:WBP UIKit Tag Base Object Reference=Get RarityTag() -> (RarityTag:WBP UIKit Tag Object Reference)
    Text:Text=Get TagPrimaryText() -> (TagPrimaryText:Text)
   ) -> (then)
   > SetLabelText!(execute, self:WBP UIKit Tag Base Object Reference=Get RarityTag(), Text:Text=Get TagSecondaryText() -> (TagSecondaryText:Text)) -> (then)
  ]}
 SetRarityTagType {nodes [
   n1 = SetRarityTagType!() -> (then, RarityTagType:E_UI_TagType Enum=NewEnumerator0) {is=title K2Node_FunctionEntry}
   > n2 = Set RarityTagType(RarityTagType:E_UI_TagType Enum=@n1) -> (Output_Get:E_UI_TagType Enum=NewEnumerator0)
   > SetTagType!(execute, self:WBP UIKit Tag Base Object Reference=Get RarityTag(), Type:E_UI_TagType Enum=@n2) -> (then)
   > Is Valid(InputObject=Get FortItemDetailsVM())
   n3 = Get FortItemDetailsVM()
   > SetRarityTagColors!(
    execute
    self:WBP UIKit Item Description Base Object Reference
    FortColorPalette:"Fort Color Palette Structure (by ref)"=GetRarityColors!(
     self
     IsSeries:bool=Get bIsSeriesItem!(self:Fort Item Details VM Object Reference=@n3) -> (bIsSeriesItem:bool)
     FortColorPalette:"Fort Color Palette Structure (by ref)"=Get RarityColorPalette!(self:Fort Item Details VM Object Reference=@n3) -> (RarityColorPalette:Fort Color Palette Structure)
    ) -> (OutputPin:Fort Color Palette Structure) {is=title K2Node_CallFunction}
   ) -> (then)
  ]}
 SetSubtitle {nodes [
   SetSubtitle!() -> (then, SubtitleText:Text) {is=title K2Node_FunctionEntry}
   > SetSubtitleText!(execute, self, InText:Text=Get SubtitleText() -> (SubtitleText:Text)) -> (then)
   > SetSubtitleFont!(
    execute
    self:WBP Dynamic Subtitle Object Reference=Get Subtitle()
    Font:Slate Font Info Structure=Get SubtitleFont() -> (SubtitleFont:Slate Font Info Structure)
   ) -> (then)
  ]}
 SetSubtitleText {nodes [
   SetSubtitleText!() -> (then, InText:Text) {is=title K2Node_FunctionEntry}
   > n1 = Branch(Condition=Is Mimosa!() -> (Bool:bool=~) {is=title K2Node_MacroInstance})
   n2 = GetMimosaName!(self) -> (NameText:Text, IsNameEmpty:bool, IsNamable:bool) {is=title K2Node_CallFunction}
   n3 = "Equal, Case Insensitive (Text)"!(
    A:"Text (by ref)"=GetItemSubTitle!(self:Fort Item Details VM Object Reference=Get FortItemDetailsVM()) -> (ReturnValue:Text)
    B:"Text (by ref)"=@n2
   ) -> (ReturnValue:bool)
   > n4 = Branch(Condition=AND Boolean!(A=NOT Boolean(A=@n3), B=NOT Boolean(A=@n2.IsNameEmpty), C:bool=@n2.IsNamable) -> (ReturnValue))
   > ShowDynamicSubtitle!(execute, self, Show:bool=true) -> (then)
   > "SetText (Text)"(self=Get Text_Subtitle(self=Get Subtitle()), InText=GetItemSubTitle(self=Get FortItemDetailsVM()))
   ShowDynamicSubtitle(execute=@n4.else, Show=false)
   n5 = TextIsEmpty(InText=Get InText() -> (InText:Text))
   n6 = Get FortItemDetailsVM()
   n7 = GetItemType(ItemDefinition=Get ItemDefinition(self=Get Item(self=@n6))) -> (ReturnValue=WorldItem)
   ShowDynamicSubtitle(execute=@n1.else, Show=AND Boolean!(A=NOT Boolean(A=@n5), B="Equal (Enum)"(A=@n7, B=SparksSong)) -> (ReturnValue))
   > "SetText (Text)"(self=Get Text_Subtitle(self=Get Subtitle()), InText=Get InText())
  ]}
 SetTitle {nodes [
   SetTitle!() -> (then, TitleText:Text) {is=title K2Node_FunctionEntry}
   n1 = Get UseOriginalDisplayName() -> (UseOriginalDisplayName:bool)
   > n2 = Branch(Condition=AND Boolean(A=Is Mimosa() -> (Bool=~), B=NOT Boolean(A=@n1)))
   n3 = GetMimosaName()
   > n4 = Branch(Condition=OR Boolean(A=NOT Boolean(A=@n3.IsNameEmpty), B=@n3.IsNamable))
   > SetTitleText!(execute, self:WBP UIKit Item Description Base Object Reference, Text:Text=@n3) -> (then)
   > SetSubtitleText() {is=title K2Node_CallFunction}
   SetTitleText(execute=[@n4.else, @n2.else], Text=Get TitleText() -> (TitleText:Text))
  ]}
 SetWidgetLayout {nodes [
   SetWidgetLayout() {is=title K2Node_FunctionEntry}
   > Parent: SetWidgetLayout() {is=title K2Node_CallParentFunction}
   n1 = SlotAsGridSlot(Widget=Get Subtitle())
   > SetRow!(execute, self:Grid Slot Object Reference=@n1, InRow:int=Get SubtitleRow() -> (SubtitleRow=0)) -> (then)
   > Set Padding!(execute, Padding:Margin Structure=Get SubtitlePadding() -> (SubtitlePadding:Margin Structure), self:Grid Slot Object Reference=@n1) -> (then, Output_Get:Margin Structure)
  ]}
 ShowDynamicSubtitle {nodes [
   n1 = ShowDynamicSubtitle!() -> (then, Show:bool=~) {is=title K2Node_FunctionEntry}
   n2 = NOT Boolean(A=Get HideSubtitle())
   > Set Visibility(Visibility=Select(Option 0=Collapsed, Option 1=SelfHitTestInvisible, Index=AND Boolean(A=@n1, B=@n2)), self=Get Subtitle())
  ]}
]
variables [
 HideSubtitle {type=boolean}
 RarityTagType {type=E_UI_TagType, default=Rarity Tag}
 Subtitle {type=WBP_DynamicSubtitle_C, default=null}
 SubtitleFont {
  type=SlateFontInfo
  default {
   bForceMonospaced=false
   bMaterialIsStencil=false
   fontMaterial=null
   fontObject=&HeadingNow
   letterSpacing=0
   monospacedWidth=1
   size=19.5
   skewAmount=0
   typefaceFontName=HeadingNow-86BoldItalic
   outlineSettings {
    bApplyOutlineToDropShadows=false
    bMiteredCorners=false
    bSeparateFillAlpha=false
    outlineMaterial=null
    outlineSize=0
    outlineColor=<rgba 0 0 0 1>
   }
  }
 }
 SubtitlePadding {type=Margin, default=<ltrb 0 8 0 0>}
 SubtitleRow {type=integer, default=2}
 Use Rich Description {type=boolean, default=true}
 UseOriginalDisplayName {type=boolean}
]
widgets {
 info {
  inheritedWidgetCount=0
  parentClass=/FortUILibrary/Systems/Items/WBP_UIKit_ItemDescription_Base.WBP_UIKit_ItemDescription_Base_C
  rootWidgetClass=null
  widgetCount=0
 }
 widgets []
}
