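# Compact item description widget, derived from WBP_UIKit_ItemDescription_Base, which it initializes on PreConstruct.
# Titles show the display name, except CosmeticMimosa items, which use the view model's item name when nameable, even if empty.
# Rarity palettes are remapped, with the first color depending on series status, and the rarity/series tag can be force-hidden by a flag.
# An age-gate info box, collapsed by default, shows when the item is age-gated.
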WidgetBlueprint /FortUILibrary/UIKit/Items/WBP_UIKit_ItemDescription_Compact
graphs [
 EventGraph {nodes [
   n1 = Event PreConstruct()
   > Parent: PreConstruct(IsDesignTime=@n1.IsDesignTime)
   > SetupWidget()
  ]}
 GetMimosaName {nodes [
   GetMimosaName() {is=title K2Node_FunctionEntry}
   > n1 = IsNameable!(execute:Exec, self:Object Reference=Get FortItemDetailsVM() -> (FortItemDetailsVM:Fort Item Details VM Object Reference)) -> (then:Exec, ReturnValue:bool) {is=title K2Node_Message}
   > n2 = Branch(Condition=@n1)
   > n3 = GetItemName!(execute:Exec, self:Object Reference=Get FortItemDetailsVM()) -> (then:Exec, ReturnValue:Text) {is=title K2Node_Message}
   > Return Node!(execute, NameText:Text=@n3, IsNameEmpty:bool=TextIsEmpty(InText=@n3), IsNamable:bool=true)
   Return Node(execute=@n2.else, IsNameEmpty=true, IsNamable=false)
  ]}
 GetRarityColors {nodes [
   n1 = GetRarityColors!() -> (then, IsSeries:bool=~, FortColorPalette:"Fort Color Palette Structure (by ref)") {is=title K2Node_FunctionEntry}
   n2 = Break Fort Color Palette!(FortColorPalette:"Fort Color Palette Structure (by ref)"=@n1.FortColorPalette) -> (
    Color1:Linear Color Structure
    Color2:Linear Color Structure
    Color3:Linear Color Structure
    Color4:Linear Color Structure
    Color5:Linear Color Structure
   ) {is=title K2Node_BreakStruct}
   > Return Node!(
    execute
    OutputPin:Fort Color Palette Structure=Make Fort Color Palette!(
     Color1:Linear Color Structure=SelectColor(A=@n2, B=@n2.Color2, bPickA=@n1)
     Color2:Linear Color Structure=@n2.Color4
     Color3:Linear Color Structure=(R=0,G=0,B=0,A=1)
     Color4:Linear Color Structure=(R=0,G=0,B=0,A=1)
     Color5:Linear Color Structure=(R=0,G=0,B=0,A=1)
    ) -> (FortColorPalette:Fort Color Palette Structure) {is=title K2Node_MakeStruct}
   )
  ]}
 IsMimosa {nodes [
   Inputs!()
   n1 = Get ItemDefinition!(self:Fort Item VM Object Reference=Get Item!(self:Fort Item Details VM Object Reference=Get FortItemDetailsVM()) -> (Item:Fort Item VM Object Reference)) -> (ItemDefinition:Fort Item Definition Object Reference)
   Outputs!(Bool:bool="Equal (Enum)"(
     A:EFortItemType Enum=GetItemType!(ItemDefinition:Item Definition Base Object Reference=@n1) -> (ReturnValue:EFortItemType Enum=WorldItem)
     B:EFortItemType Enum=CosmeticMimosa
    ))
  ]}
 OnItemDetailsSet {nodes [OnItemDetailsSet!() -> (then, FortItemDetailsVM:Fort Item Details VM Object Reference) {is=title K2Node_FunctionEntry}]}
 SetDisplayTitle {nodes [
   SetDisplayTitle!() -> (then, DisplayNameText:Text) {is=title K2Node_FunctionEntry}
   > n1 = Branch(Condition=Is Mimosa!() -> (Bool:bool=~) {is=title K2Node_MacroInstance})
   n2 = GetMimosaName!(self) -> (NameText:Text, IsNameEmpty:bool, IsNamable:bool) {is=title K2Node_CallFunction}
   > n3 = Branch(Condition=OR Boolean(A=NOT Boolean(A=@n2.IsNameEmpty), B=@n2.IsNamable))
   > SetTitleText!(execute, self:WBP UIKit Item Description Base Object Reference, Text:Text=@n2) -> (then)
   SetTitleText(execute=[@n3.else, @n1.else], Text=Get DisplayNameText() -> (DisplayNameText:Text))
  ]}
 SetupWidget {nodes [
   SetupWidget() {is=title K2Node_FunctionEntry}
   > InitializeWidget(self:WBP UIKit Item Description Base Object Reference)
  ]}
 ShowRaritySeriesTag {nodes [
   n1 = ShowRaritySeriesTag!() -> (then, ShowRarity:bool=~) {is=title K2Node_FunctionEntry}
   > SetRarityTagVisibility!(
    execute
    self:WBP UIKit Item Description Base Object Reference
    ShowTag:bool=Select(Option 0:bool=@n1, Option 1:bool=false, Index=Get HideRarityAndSeriesTag() -> (HideRarityAndSeriesTag:bool)) -> (ReturnValue:bool)
   ) -> (then)
  ]}
 UpdateAgeGateWarningVisibility {nodes [
   n1 = UpdateAgeGateWarningVisibility!() -> (then, IsAgeGated:bool=~) {is=title K2Node_FunctionEntry}
   > SetVisibility(
    self=Get WBP_AgeGate_Warning() -> (WBP_AgeGate_Warning:WBP UIKit Info Box Object Reference)
    InVisibility=Select(Option 0:ESlateVisibility Enum=Collapsed, Option 1:ESlateVisibility Enum=HitTestInvisible, Index=@n1) -> (ReturnValue:ESlateVisibility Enum)
   )
  ]}
]
variables [
 CustomModalWidget {type=Class, default=null}
 HideRarityAndSeriesTag {type=boolean}
 Input Action Row {type=DataTableRowHandle, default {dataTable=null, rowName=None}}
 ItemName {type=string, default=""}
]
widgets {
 info {
  inheritedWidgetCount=0
  namedSlotCount=1
  parentClass=/FortUILibrary/Systems/Items/WBP_UIKit_ItemDescription_Base.WBP_UIKit_ItemDescription_Base_C
  rootWidgetClass=null
  widgetCount=1
 }
 widgets [WBP_AgeGate_Warning {
   widgetClassPath=/FortUILibrary/UIKit/InfoBox/WBP_UIKit_InfoBox
   bIsVariable=true
   widget=WBP_AgeGate_Warning
   properties {
    iconMaterial=/FortUILibrary/Textures/MaterialInstances/MI_UIKit_Icon_Information
    previewText="Unlocking these rewards requires access to <text color=\"FFF\" fontface=\"76 Bold\">teen rated experiences.</>"
    use Preview Text=true
    visibility=Collapsed
   }
  }]
}
